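# Niagara trail for the DarkNinja glider: a purple ribbon that stretches and brightens with the smoothed forward dot, with front-end overrides.
# Trail bursts 20 persistent particles, ordered by normalized index, that ease toward offsets from sampled skeletal mesh positions and lag against owner velocity.
# Curves along the ribbon shape color and width, forward dot drives UV scroll speed, and two ribbon renderers draw it.
# Glow and Light emitters list no renderers.
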NiagaraSystem /BRCosmetics/Gadgets/Assets/VinderTech_GliderChute/Glider_DarkNinja/FX/NS_Glider_Trail_DarkNinja
dependencies {
 usedDataInterfaces [&NiagaraDataInterfaceColorCurve, &NiagaraDataInterfaceCurve, &NiagaraDataInterfaceVectorCurve]
 usedDynamicInputs [&VectorFromCurve, &Waveform, &ClampFloat]
 usedModules [
  &self:SystemSpawnScript.SystemScriptSource.SystemScriptGraph.NiagaraNodeAssignment_2.SetVariables_#1
  &SystemState
  &self:SystemSpawnScript.SystemScriptSource.SystemScriptGraph.NiagaraNodeAssignment_0.SetVariables_#2
  &self:SystemSpawnScript.SystemScriptSource.SystemScriptGraph.NiagaraNodeAssignment_1.SetVariables_#3
 ]
}
emitters {
 Glow {rendererClasses [], renderers []}
 Light {rendererClasses [], renderers []}
 Trail {
  emitterUpdateScript [
   &EmitterState(
    Inactive Response=::NewEnumerator1
    Life Cycle Mode=::NewEnumerator1
    Loop Behavior=::NewEnumerator1
    Loop Duration Mode=::NewEnumerator1
    UseLoopDelay=false
   )
   &SpawnBurst_Instantaneous(Spawn Count=20, Spawn Time=0.75, Use Spawn Probability=false, UseLoopCountLimit=false)
  ]
  particleSpawnScript [
   &InitializeParticle(Lifetime=5, Ribbon Width=10, Ribbon Width Mode=::NewEnumerator1, UsePositionOffset=false, Write Material Random=true)
   /SetVariables(Particles.NormalizedRibbonPos:NiagaraFloat=&OneMinusFloat(Float=&ReturnNormalizedExecIndex())) {isSetParameters=true}
   /SetVariables(Particles.RibbonLinkOrder:NiagaraFloat=$Particles.NormalizedRibbonPos) {isSetParameters=true}
   /SetVariables(Particles.GlowColor=&ColorFromCurve(
     ColorCurve=DataInterface() {propertyValues {
       AlphaCurve {keys [<at 0 1 RCIM_Cubic>,<at 0.6603 1>,<at 1 0>]}
       BlueCurve {keys [<at 0 1 RCIM_Cubic>,<at 0.3054 0.575>]}
       GreenCurve {keys [<at 0 1 RCIM_Cubic>,<at 0.3054 0.575>]}
       RedCurve {keys [<at 0 1 RCIM_Cubic>,<at 0.3054 0.575>]}
      }}
     CurveIndex=$Particles.NormalizedRibbonPos
    )) {isSetParameters=true}
   /SetVariables() {isSetParameters=true}
  ]
  particleUpdateScript [
   &ParticleState(Kill Particles When Lifetime Has Elapsed=false)
   /SetVariables(Particles.UVOffset:NiagaraFloat=&Add_Float(
     A=$Particles.UVOffset
     B=&Multiply_Float(A=$Engine.DeltaTime, B=&Lerp_Float(A=-0.1, B=-1.5, Alpha=$System.SmoothedForwardDot))
    )) {isSetParameters=true}
   &Color(
    Color=&ColorFromCurve(
     ColorCurve=DataInterface() {propertyValues {
       AlphaCurve {keys [<at 0 0>,<at 0.0975 1 RCIM_Cubic>,<at 0.8624 0 RCIM_Cubic>]}
       BlueCurve {keys [<at 0.0123 1>,<at 1 1>]}
       GreenCurve {keys [<at 0.0123 0>,<at 1 0.0848>]}
       RedCurve {keys [<at 0.0123 0.7253>,<at 1 0.3921>]}
      }}
     CurveIndex=$Particles.NormalizedRibbonPos
    )
    Scale Alpha=&Multiply_Float(A=1, B=$System.SmoothedForwardDot)
   )
   /SetVariables(Particles.SpawnAlpha:NiagaraFloat=&FloatFromCurve(FloatCurve=DataInterface() {propertyValues {Curve {keys [<at 0 0>,<at 1 1>]}}}, CurveIndex=$Particles.Age)) {isSetParameters=true}
   /SetVariables(Particles.GoalPos:NiagaraPosition=&AddVectorToPosition(
     Position=$Particles.SkeletalMeshLocation.SampledPosition
     Vector=&Lerp_Vector(
      A=<0 0 0>
      B=&Lerp_Vector(
       A=<0 0 0>
       B=&Lerp_Vector(
        A=<0 0 0>
        B=<0 400 0>
        Alpha=&FN_FrontEndFloatModifier(Gameplay=$System.SmoothedForwardDot, Front End=2)
       )
       Alpha=$Particles.SpawnAlpha
      )
      Alpha=$Particles.NormalizedRibbonPos
     )
    )) {isSetParameters=true}
   &InheritSourceMovement(
    Applied Position Scale=&VectorFromFloat(Value=&Multiply_Float(
      A=-0.5
      B=&FloatFromCurve(
       FloatCurve=DataInterface() {propertyValues {Curve {keys [
           {arriveTangent=-0.0194, interpMode=RCIM_Cubic, leaveTangent=-0.0194, tangentMode=RCTM_User, time=0.2, value=0}
           {arriveTangent=1.6438, interpMode=RCIM_Cubic, leaveTangent=1.6438, tangentMode=RCTM_User, time=1, value=1}
          ]}}}
       CurveIndex=$Particles.NormalizedRibbonPos
      )
     ))
    ApplyForce=false
    ApplyPosition=true
    ApplyPositionFalloff=false
    Source Velocity:Vector3f=&TransformVector(Vector=$Engine.Owner.Velocity, Source Space=::World, Destination Space=::Local)
   )
   &SolveForcesAndVelocity(Clamp Velocity=false, Limit Acceleration=false)
   /SetVariables(Particles.Position=&Lerp_Vector(
     A=Linked() {linkedVariable {name=Particles.Position, type=NiagaraPosition}}
     B=Linked() {linkedVariable {name=Particles.GoalPos, type=NiagaraPosition}}
     Alpha=&Multiply_Float(A=$Engine.DeltaTime, B=10)
    )) {isSetParameters=true}
   &ScaleRibbonWidth(Ribbon Width Scale=&FloatFromCurve(
     FloatCurve=DataInterface() {propertyValues {Curve {keys [
         {arriveTangent=0.5751, interpMode=RCIM_Cubic, leaveTangent=0.5751, tangentMode=RCTM_User, time=0.25, value=0.45}
         {interpMode=RCIM_Cubic, tangentMode=RCTM_User, time=0.6005, value=1.4377}
         {arriveTangent=-4.3814, interpMode=RCIM_Cubic, leaveTangent=-4.3814, tangentMode=RCTM_User, time=1, value=0.5}
        ]}}}
     CurveIndex=$Particles.NormalizedRibbonPos
     Scale Curve=$Particles.SpawnAlpha
    ))
   /SetVariables(Particles.GlowWidth:NiagaraFloat=&Multiply_Float(A=$Particles.RibbonWidth, B=2)) {isSetParameters=true}
   &DynamicMaterialParameters(
    Index 0 Param 1=$Particles.UVOffset
    Index 0 Param 2=&FN_FrontEndFloatModifier(Gameplay=$System.SmoothedForwardDot, Front End=2)
    Module.Index0Float0Write=true
    Module.Index0Float1Write=true
    Module.Index0Float2Write=true
    Module.Index0Float3Write=true
   )
   /SetVariables(Particles.GlowColor=&Multiply_LinearColorByFloat(Linear Color=<rgba 0.5207 0.0001 1 1>, Float=$System.SmoothedForwardDot)) {isSetParameters=true}
  ]
  rendererClasses [&NiagaraRibbonRendererProperties]
  renderers [
   {rendererClass=&NiagaraRibbonRendererProperties}
   {rendererClass=&NiagaraRibbonRendererProperties, rendererIndex=1}
  ]
 }
}
summary {
 systemName=NS_Glider_Trail_DarkNinja
 emitters [
  {emitterName=Glow, rendererClasses []}
  {emitterName=Light, rendererClasses []}
  {emitterName=Trail, rendererClasses [&NiagaraRibbonRendererProperties]}
 ]
 userVariables [
  User.bIsFrontEnd {type=NiagaraBool}
  User.bIsFrontEndPreview {type=NiagaraBool}
  User.ForwardDot {type=NiagaraFloat, defaultValue=0, description=""}
  User.SkeletalMesh {
   type=NiagaraDataInterfaceSkeletalMesh
   defaultValue {
    struct=NiagaraExt_VariableValue_DataInterface
    value {dataInterface=&self:NiagaraDataInterfaceSkeletalMesh_0}
   }
  }
 ]
}
